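# Niagara system for the Aurawater vehicle boost cosmetic, a glowing water-themed trail emitted from each boost socket on the car's skeletal mesh.
# Three ribbon emitters form the streaks, two writing a sine-wave WPO position for ripples, alongside rearward star sprites and a socket-anchored glow of meshes and a sprite.
# Painted-variant modules set colors per layer; ribbons switch to default colors or curves when unpainted.
# User.DTCarOffset shifts every emitter's spawn position.
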NiagaraSystem /VehicleCosmetics/Boosts/Boost_Aurawater/NS/NS_Boost_Aurawater
dependencies {
 usedDataInterfaces [
  &NiagaraDataInterfaceCurve
  &NiagaraDataInterfaceColorCurve
  &NiagaraDataInterfaceArrayColor
  &NiagaraDataInterfaceArrayFloat
  &NiagaraDataInterfaceArrayFloat3
  &NiagaraDataInterfaceSkeletalMesh
 ]
 usedDynamicInputs []
 usedModules [
  &self:SystemSpawnScript.SystemScriptSource.SystemScriptGraph.NiagaraNodeAssignment_0.SetVariables_#1
  &SystemState
 ]
}
emitters {
 StarSprites001 {
  emitterUpdateScript [
   &EmitterState()
   &VC_FilteredSocketCount()
   &SpawnRate(SpawnRate=25, Use Spawn Probability=false)
  ]
  particleSpawnScript [
   &VC_PaintedVariantsOverride(
    Black=<rgba 0.2 0.2 0.2 1>
    BurntSienna=<rgba 0.19 0.0475 0 1>
    Cobalt=<rgba 0.2 0.4 1 1>
    Crimson=<rgba 0.5 0 0.0136 1>
    DefaultColor=&SelectLinearColorFromArray(
     Array Sampling Mode=::NewEnumerator0
     Direct Array Index=?
     Color Selection Array=DataInterface() {propertyValues {ColorData [<rgba 1 1 1 1>, <rgba 0.2 0.8667 1 1>]}}
     Random Seed=0
    )
    ForestGreen=<rgba 0 0.6 0.0303 1>
    Grey=<rgba 0.4 0.4 0.4 1>
    Lime=<rgba 0.4667 0.7 0 1>
    Orange=<rgba 0.7 0.2917 0 1>
    Pink=<rgba 1 0.3 0.7455 1>
    SkyBlue=<rgba 0.1 0.6988 1 1>
   )
   &InitializeParticle(
    Color=$Particles.Color
    Lifetime Max=0.9
    Lifetime Min=0.8
    Lifetime Mode=::NewEnumerator1
    Sprite Size Mode=::NewEnumerator1
    Uniform Sprite Size Max=0.2
    Uniform Sprite Size Min=0.1
    UsePositionOffset=false
    Write Material Random=true
   )
   &PartitionParticles(Number Of Partitions=$Emitter.PortCount)
   &SkeletalMeshLocation(
    Bone / Socket Index=$Particles.PartitionParticles.Partition
    Bone / Socket Sampling Mode=::NewEnumerator5
    MaskSpawnGroup=false
    Mesh Sampling Type=::NewEnumerator6
    Sampled Position Offset=&AddVectorToPosition(Position=<-80 0 0>, Vector=$User.DTCarOffset)
    Skeletal Mesh=$System.SkeletalMesh
   )
   &ShapeLocation(MaskSpawnGroup=false, Override Seed=false, Sphere Radius=30, Sphere Surface Distribution=1, UseOwnerScale=true)
   &AddVelocity(Velocity=<-650 0 0>)
   /SetVariables(Particles.VisibilityTag=$Emitter.SpawnGroup) {isSetParameters=true}
  ]
  particleUpdateScript [
   &ParticleState()
   &ScaleSpriteSize(
    Scale Sprite Size Mode=::NewEnumerator0
    Uniform Scale Factor=&FloatFromCurve(
     FloatCurve=DataInterface() {propertyValues {Curve {
        defaultValue=1
        keys [{interpMode=RCIM_Cubic, tangentMode=RCTM_User, time=0.25, value=1}, {interpMode=RCIM_Cubic, tangentMode=RCTM_User, time=1, value=0.5}]
       }}}
     Scale Curve=40
    )
   )
   &SolveForcesAndVelocity(Clamp Velocity=false, Limit Acceleration=false)
   /SetVariables(Particles.AURA_SIZE:Vector2f=&Multiply_Vector2D_ByFloat(Float=5, Vector2D=$Particles.SpriteSize)) {isSetParameters=true}
   &ScaleColor(
    Scale Alpha=&FloatFromCurve(FloatCurve=DataInterface() {propertyValues {Curve {
        defaultValue=1
        keys [
         {interpMode=RCIM_Cubic, tangentMode=RCTM_User, time=0, value=0.15}
         {arriveTangent=2.5761, interpMode=RCIM_Cubic, leaveTangent=-3.2037, tangentMode=RCTM_Break, time=0.5, value=1}
         <at 1.0018 0 RCIM_Cubic>
        ]
       }}})
    ScaleA=true
    ScaleRGB=true
   )
  ]
  rendererClasses [&NiagaraSpriteRendererProperties]
  renderers [
   {rendererClass=&NiagaraSpriteRendererProperties}
   {rendererClass=&NiagaraSpriteRendererProperties, rendererIndex=1}
  ]
 }
 mesh_glow002 {
  emitterUpdateScript [
   &EmitterState(
    Inactive Response=::NewEnumerator1
    Life Cycle Mode=::NewEnumerator1
    Loop Behavior=::NewEnumerator1
    Loop Duration Mode=::NewEnumerator1
    UseLoopDelay=false
   )
   &VC_FilteredSocketCount()
   &SpawnBurst_Instantaneous(Spawn Count=2, Use Spawn Probability=false, UseLoopCountLimit=false)
   &SpawnRate(SpawnRate=2.2, Use Spawn Probability=false) {enabled=false}
  ]
  particleSpawnScript [
   &VC_PaintedVariantsOverride(
    Black=<rgba 0 0 0 1.5>
    BurntSienna=<rgba 0.02 0.0062 0.0025 2>
    Cobalt=<rgba 0 0.0146 0.35 1.5>
    Crimson=<rgba 0.05 0 0.0014 2>
    DefaultColor=<rgba 0 0.0758 0.35 1>
    ForestGreen=<rgba 0 0.05 0.0025 2>
    Grey=<rgba 0.03 0.03 0.03 3>
    Lime=<rgba 0.0375 0.15 0 2>
    Orange=<rgba 0.08 0.0107 0 2>
    Pink=<rgba 0.09 0.027 0.0742 1>
    Purple=<rgba 0.0125 0 0.05 2>
    R1 Black:LinearColor=<rgba 0.15 0.15 0.15 0.5>
    R1 BurntSienna:LinearColor=<rgba 0.25 0.0333 0 1.5>
    R1 Cobalt:LinearColor=<rgba 0.055 0.1787 0.55 1.5>
    R1 Crimson:LinearColor=<rgba 1 0 0 1>
    R1 DefaultColor=<rgba 0.5165 1 1 1>
    R1 Grey:LinearColor=<rgba 0.2 0.2 0.2 2>
    R1 Lime:LinearColor=<rgba 0.5 1 0 1.5>
    R1 Orange:LinearColor=<rgba 1 0.1667 0 1>
    R1 Pink:LinearColor=<rgba 1 0.3 0.7455 1>
    R1 Purple:LinearColor=<rgba 0.15 0 1 2>
    R1 Saffron:LinearColor=<rgba 1 0.5667 0 1.5>
    R1 SkyBlue:LinearColor=<rgba 0.1 0.4 1 1>
    R1 TitaniumWhite:LinearColor=<rgba 1 1 1 1>
    R2 Black:LinearColor=<rgba 0.07 0.07 0.07 1>
    R2 BurntSienna:LinearColor=<rgba 0.15 0.03 0 1>
    R2 Cobalt:LinearColor=<rgba 0.04 0.1 0.4 1>
    R2 Crimson:LinearColor=<rgba 0.45 0 0 1>
    R2 DefaultColor:LinearColor=<rgba 0.3126 0.5619 0.6783 1>
    R2 ForestGreen:LinearColor=<rgba 0 0.25 0.0126 1>
    R2 Grey:LinearColor=<rgba 0.15 0.15 0.15 1>
    R2 Lime:LinearColor=<rgba 0.325 0.65 0 1>
    R2 Orange:LinearColor=<rgba 0.6 0.075 0 1.5>
    R2 Pink:LinearColor=<rgba 0.45 0.1479 0.4298 1>
    R2 Purple:LinearColor=<rgba 0.0675 0 0.45 1>
    R2 Saffron:LinearColor=<rgba 0.4 0.2467 0 1>
    R2 SkyBlue:LinearColor=<rgba 0.12 0.44 0.6 1>
    R2 TitaniumWhite:LinearColor=<rgba 0.233 0.233 0.233 1>
    R3 Black:LinearColor=<rgba 0.12 0.12 0.12 2>
    R3 BurntSienna:LinearColor=<rgba 0.4 0.0667 0 2>
    R3 Cobalt:LinearColor=<rgba 0.24 0.48 1.2 1>
    R3 Crimson:LinearColor=<rgba 1.5 0 0.0406 1>
    R3 DefaultColor:LinearColor=<rgba 0.5 0.875 1 0.9>
    R3 ForestGreen:LinearColor=<rgba 0 1 0.0505 1>
    R3 Grey:LinearColor=<rgba 0.35 0.35 0.35 1>
    R3 Lime:LinearColor=<rgba 0.5 1 0 1.5>
    R3 Orange:LinearColor=<rgba 1 0.1667 0 1.6>
    R3 Pink:LinearColor=<rgba 1 0.3 0.9417 1.4>
    R3 Purple:LinearColor=<rgba 0.17 0 0.85 2>
    R3 SkyBlue:LinearColor=<rgba 0.05 0.7625 1 1.35>
    Saffron=<rgba 0.04 0.0267 0.02 2>
    SkyBlue=<rgba 0 0.06 0.2 2>
    TitaniumWhite=<rgba 0.05 0.05 0.05 2>
    Use R1 Painted Overrides=true
    Use R2 Painted Overrides=true
    Use R3 Painted Overrides=true
   )
   /SetVariables(Particles.VisibilityTag=$Emitter.SpawnGroup) {isSetParameters=true}
   &InitializeParticle(
    Color=$Particles.Color
    Mesh Scale=<0.1 0.07 0.08>
    Mesh Scale Mode=::NewEnumerator4
    Sprite Rotation Mode=::NewEnumerator2
    UsePositionOffset=false
    Write Material Random=true
   )
   &PartitionParticles(Number Of Partitions=$Emitter.PortCount)
   &SkeletalMeshLocation(
    Bone / Socket Index=$Particles.PartitionParticles.Partition
    Bone / Socket Sampling Mode=::NewEnumerator5
    MaskSpawnGroup=false
    Mesh Sampling Type=::NewEnumerator6
    Sampled Position Offset=&AddVectorToPosition(Position=<-50 0 0>, Vector=$User.DTCarOffset)
    Skeletal Mesh=$System.SkeletalMesh
   )
   &AddVelocity(Velocity=<0 0 0>) {enabled=false}
   &InitialMeshOrientation(
    Mesh Axis To Orient=<0 0 0>
    Mesh Orientation Mode=::NewEnumerator2
    Orientation Vector=<0 0 1>
    Rotation=&SelectVectorFromArray(
     Direct Array Index=$Particles.PartitionParticles.Partition
     Vector Selection Array=DataInterface() {propertyValues {FloatData [<0 0 0>, <0 0 0>]}}
    )
    Rotation Coordinate Space=::NewEnumerator2
    Use Rotation Vector=true
   )
  ]
  particleUpdateScript [
   &ParticleState(Kill Particles When Lifetime Has Elapsed=false, Loop Particles Lifetime=true)
   &SolveForcesAndVelocity(Clamp Velocity=false, Limit Acceleration=false) {enabled=false}
   /SetVariables(
    Particles.Fresnel Color:LinearColor=&MultiplyLinearColor(
     Linear Color=$Particles.Initial.R2 Color
     Scale Factor=&MakeVector4FromVectorAndFloat("Vector (RGB)"=<1 1 1>, "Float (Alpha)"=1)
    )
    Particles.Waves Color:LinearColor=&MultiplyLinearColor(
     Linear Color=$Particles.Initial.R3 Color
     Scale Factor=&MakeVector4FromVectorAndFloat("Vector (RGB)"=<1 1 1>, "Float (Alpha)"=0.1)
    )
   ) {isSetParameters=true}
  ]
  rendererClasses [&NiagaraMeshRendererProperties, &NiagaraSpriteRendererProperties]
  renderers [
   {rendererClass=&NiagaraMeshRendererProperties, rendererIndex=0}
   {rendererClass=&NiagaraMeshRendererProperties, rendererIndex=1}
   {rendererClass=&NiagaraMeshRendererProperties, rendererIndex=2}
   {rendererClass=&NiagaraMeshRendererProperties, rendererIndex=3}
   {rendererClass=&NiagaraSpriteRendererProperties, rendererIndex=4}
  ]
 }
 ribbon {
  emitterUpdateScript [
   &EmitterState(Inactive Response=::NewEnumerator1, Life Cycle Mode=::NewEnumerator1, Loop Duration=1, UseLoopDelay=false)
   &VC_FilteredSocketCount()
   &VC_RateQualityHandler()
   &VC_SpawnRibbon_v1(Number of Ribbons=$Emitter.PortCount, Spawn Time=-1, SpawnRate=60)
  ]
  particleSpawnScript [
   &VC_PaintedVariantsOverride(
    Black=<rgba 0.01 0.01 0.01 2>
    BurntSienna=<rgba 0.15 0.04 0 1>
    Cobalt=<rgba 0.006 0.0256 0.3 1.5>
    Crimson=<rgba 0.0816 0 0.0022 1.5>
    ForestGreen=<rgba 0 0.0608 0.0031 1.5>
    Grey=<rgba 0.06 0.06 0.06 1>
    Lime=<rgba 0.0334 0.1565 0.025 1.5>
    Orange=<rgba 0.2205 0.0372 0 1.5>
    Pink=<rgba 0.3003 0.0771 0.2854 1>
    Purple=<rgba 0.015 0 0.09 2>
    R1 Black:LinearColor=<rgba 0.03 0.03 0.03 1>
    R1 BurntSienna:LinearColor=<rgba 0.08 0.0133 0 1>
    R1 Cobalt:LinearColor=<rgba 0.045 0.1057 0.45 1>
    R1 ForestGreen:LinearColor=<rgba 0 0.2319 0.004 1>
    R1 Grey:LinearColor=<rgba 0.12 0.12 0.12 1>
    R1 Orange:LinearColor=<rgba 0.6 0.08 0 1>
    R1 Pink:LinearColor=<rgba 0.32 0.112 0.32 1>
    R1 Purple:LinearColor=<rgba 0.035 0 0.3 1>
    R1 Saffron:LinearColor=<rgba 0.5 0.2917 0 1>
    R1 SkyBlue:LinearColor=<rgba 0.1226 0.2854 0.5 1>
    R1 TitaniumWhite:LinearColor=<rgba 0.45 0.45 0.45 1>
    Saffron=<rgba 0.12 0.08 0 1.25>
    SkyBlue=<rgba 0 0.0625 0.25 1.5>
    TitaniumWhite=<rgba 0.1406 0.1406 0.1406 1>
    Use R1 Painted Overrides=true
   )
   &VC_InitializeRibbon_v1(
    Color=$Particles.Color
    Custom Offset=true
    Enable Skeletal Position=true
    Lifetime=0.75
    Local Vector Offset=&AddVectorToPosition(Position=<-30 0 0>, Vector=$User.DTCarOffset)
    Offset Amount=1
    Ribbon Width=40
    Write Material Random=false
   )
   &AddVelocity(Velocity=<-1750 0 0>) {module=AddVelocity001}
   /SetVariables(Particles.color_lines:LinearColor=&MakeCustomLinearColorFromBool(Bool=$User.bIsPainted, True Color=$Particles.R1 Color, False Color=<rgba 1 1 1 1>)) {isSetParameters=true, module=SetVariables_#2}
  ]
  particleUpdateScript [
   &ParticleState()
   /SetVariables(Particles.curve_color_lines:LinearColor=&MultiplyLinearColor(
     Linear Color=$Particles.color_lines
     Scale Factor=&MakeVector4_FromLinearColor(LinearColor=&MakeLinearColorFromVectorAndFloat(
       "Vector (RGB)"=<1 1 1>
       "Float (Alpha)"=&FloatFromCurve(FloatCurve=DataInterface() {propertyValues {Curve {keys [
            {arriveTangent=8.3804, interpMode=RCIM_Cubic, leaveTangent=8.3804, tangentMode=RCTM_User, time=0.01, value=0}
            {interpMode=RCIM_Cubic, tangentMode=RCTM_Break, time=0.5, value=1}
            {interpMode=RCIM_Cubic, tangentMode=RCTM_Break, time=1, value=0}
           ]}}})
      ))
    )) {isSetParameters=true, module=SetVariables_#3}
   &Color(Color=&MultiplyLinearColor(
     Linear Color=&MakeCustomLinearColorFromBool(
      Bool=$User.bIsPainted
      True Color=$Particles.Color
      False Color=&ColorFromCurve(ColorCurve=DataInterface() {propertyValues {
         bExposeCurve=true
         AlphaCurve {keys [<at 0 0>,<at 0.1696 1>,<at 1 1 RCIM_Cubic>]}
         BlueCurve {keys [<at 0.2204 0.2691>, <at 0.3397 0.252>, {interpMode=RCIM_Cubic, leaveTangent=-0.598, time=0.6489, value=0.7}]}
         GreenCurve {keys [<at 0.2204 0.0599>, <at 0.3397 0.2164>, {interpMode=RCIM_Cubic, leaveTangent=0.6578, time=0.6489, value=1}]}
         RedCurve {keys [<at 0.2204 0>, <at 0.3397 0.1291>, {interpMode=RCIM_Cubic, leaveTangent=0.3972, time=0.6489, value=0.4}]}
        }})
     )
     Scale Factor=/Niagara/DynamicInputs/TypeConversions/MakeCustomVector4FromBool(
      Bool:NiagaraBool=$User.bIsFrontendPreview
      True Vector4:Vector4f=<1 1 1 1>
      False Vector4:Vector4f=&MakeCustomVector4FromBool(
       Bool:NiagaraBool=$User.bIsPainted
       True Vector4:Vector4f=<1 1 1 1>
       False Vector4:Vector4f=&MakeVector4FromVectorAndFloat("Vector (RGB)"=&VectorFromFloat(Value=1.5), "Float (Alpha)"=2)
      )
     )
    )) {module=Color001}
   &InheritSourceMovement(
    Applied Position Scale=&Multiply_VectorByFloat(Vector=<1 1 1>, Float=1)
    ApplyForce=false
    ApplyPosition=true
    ApplyPositionFalloff=false
   ) {module=InheritSourceMovement001}
   &SolveForcesAndVelocity(Clamp Velocity=false, Limit Acceleration=false)
   &VC_UpdateRibbon_v1(Enable Skeletal Position=true)
   &ScaleRibbonWidth(Ribbon Width Scale=&FloatFromCurve(
     FloatCurve=DataInterface() {propertyValues {Curve {keys [
         {arriveTangent=0.3269, interpMode=RCIM_Cubic, time=0, value=1.1}
         {arriveTangent=-0.4057, interpMode=RCIM_Cubic, leaveTangent=-0.4057, tangentMode=RCTM_User, time=0.2, value=0.6973}
         {arriveTangentWeight=1.0346, time=0.9195, value=0.18}
        ]}}}
     Scale Curve=1.2
    ))
   /SetVariables(Particles.WPO:NiagaraPosition=&AddVectorToPosition(
     Position=$Particles.Position
     Vector=&TransformVector(
      Vector=&MakeVector(
       X=0
       Y=0
       Z=&Sine(
        Normalized Angle=&Add_Float(
         A=&Multiply_Float(A=$Particles.NormalizedAge, B=1)
         B=&Multiply_Float(A=$Emitter.LoopedAge, B=-0.5)
        )
        Period=&MaskFloatBySpawnGroup(Masked Float=0.5, Passthrough Float=-0.5)
        Scale=&FloatFromCurve(
         FloatCurve=DataInterface() {propertyValues {Curve {
            defaultValue=1
            keys [{interpMode=RCIM_Cubic, tangentMode=RCTM_User, time=0, value=0}, {interpMode=RCIM_Cubic, tangentMode=RCTM_User, time=0.689, value=1}]
           }}}
         Scale Curve=3
        )
        Bias=0
       )
      )
      Destination Space=::Simulation
      Apply Scale=true
     )
    )) {isSetParameters=true}
  ]
  rendererClasses [&NiagaraRibbonRendererProperties]
  renderers [
   {rendererClass=&NiagaraRibbonRendererProperties}
   {rendererClass=&NiagaraRibbonRendererProperties, rendererIndex=1}
  ]
 }
 ribbon002 {
  emitterUpdateScript [
   &EmitterState(Inactive Response=::NewEnumerator1, Life Cycle Mode=::NewEnumerator1, Loop Duration=1, UseLoopDelay=false)
   &VC_FilteredSocketCount()
   &VC_RateQualityHandler()
   &VC_SpawnRibbon_v1(Number of Ribbons=$Emitter.PortCount, Spawn Time=-1, SpawnRate=60)
  ]
  particleSpawnScript [
   &VC_PaintedVariantsOverride(
    Black=<rgba 0.2656 0.2656 0.2656 1>
    BurntSienna=<rgba 0.224 0.0691 0.0276 1>
    Cobalt=<rgba 0.0625 0.124 0.8 1>
    Crimson=<rgba 1 0 0 1>
    Grey=<rgba 0.6163 0.6163 0.6163 1>
    Lime=<rgba 0.5 1 0 1>
    Orange=<rgba 0.52 0.104 0 1>
    Pink=<rgba 0.5052 0.2223 0.4023 1>
    Purple=<rgba 0.1125 0 0.45 1>
    Saffron=<rgba 1 0.8333 0 0.5>
    SkyBlue=<rgba 0.27 0.545 0.6 1>
   )
   &VC_InitializeRibbon_v1(
    Color=$Particles.Color
    Custom Offset=true
    Enable Skeletal Position=true
    Lifetime=0.6
    Local Vector Offset=&AddVectorToPosition(Position=<-50 0 0>, Vector=$User.DTCarOffset)
    Offset Amount=1
    Ribbon Width=2
    Write Material Random=false
   )
   &AddVelocity(Velocity=<1750 0 0>, Velocity Speed Scale=-1) {module=AddVelocity001}
  ]
  particleUpdateScript [
   &Color(Color=&MakeCustomLinearColorFromBool(Bool=$User.bIsPainted, True Color=$Particles.Color, False Color=<rgba 1 1 1 1>))
   &ParticleState()
   &InheritSourceMovement(
    Applied Position Scale=&Multiply_VectorByFloat(Vector=<1 1 1>, Float=1)
    ApplyForce=false
    ApplyPosition=true
    ApplyPositionFalloff=false
   )
   &SolveForcesAndVelocity(Clamp Velocity=false, Limit Acceleration=false)
   &VC_UpdateRibbon_v1(Enable Skeletal Position=true)
   &ScaleRibbonWidth()
  ]
  rendererClasses [&NiagaraRibbonRendererProperties]
  renderers [
   {rendererClass=&NiagaraRibbonRendererProperties}
   {rendererClass=&NiagaraRibbonRendererProperties, rendererIndex=1}
  ]
 }
 ribbon003 {
  emitterUpdateScript [
   &EmitterState(Inactive Response=::NewEnumerator1, Life Cycle Mode=::NewEnumerator1, Loop Duration=1, UseLoopDelay=false)
   &VC_FilteredSocketCount()
   &VC_RateQualityHandler()
   &VC_SpawnRibbon_v1(Number of Ribbons=$Emitter.PortCount, Spawn Time=-1, SpawnRate=60)
  ]
  particleSpawnScript [
   &VC_PaintedVariantsOverride(
    Black=<rgba 0.07 0.07 0.07 0.8841>
    BurntSienna=<rgba 0.2 0.0367 0 0.7>
    Cobalt=<rgba 0.066 0.187 0.55 1>
    ForestGreen=<rgba 0 0.519 0 1>
    Grey=<rgba 0.4774 0.4774 0.4774 0.3>
    Lime=<rgba 0.2667 0.8 0 1>
    Orange=<rgba 0.7 0.14 0 1>
    Pink=<rgba 0.528 0.1969 0.5469 0.75>
    Purple=<rgba 0.0587 0 0.32 1>
    Saffron=<rgba 0.9 0.5653 0.027 0.8>
    SkyBlue=<rgba 0.1277 0.5681 0.7896 0.7>
    TitaniumWhite=<rgba 0.75 0.75 0.75 0.7>
    Use R1 Painted Overrides=true
   )
   &VC_InitializeRibbon_v1(
    Color=$Particles.Color
    Custom Offset=true
    Enable Skeletal Position=true
    Lifetime=0.45
    Local Vector Offset=&AddVectorToPosition(Position=<-25 0 0>, Vector=$User.DTCarOffset)
    Offset Amount=1
    Ribbon Width=40
    Write Material Random=false
   )
   &AddVelocity(Velocity=<-1499 0 0>) {module=AddVelocity001}
   /SetVariables(Particles.color_lines:LinearColor=&MakeCustomLinearColorFromBool(Bool=$User.bIsPainted, True Color=$Particles.R1 Color, False Color=<rgba 1 1 1 1>)) {isSetParameters=true, module=SetVariables_#2}
  ]
  particleUpdateScript [
   &ParticleState()
   /SetVariables(Particles.curve_color_lines:LinearColor=&MultiplyLinearColor(
     Linear Color=$Particles.color_lines
     Scale Factor=&MakeVector4_FromLinearColor(LinearColor=&MakeLinearColorFromVectorAndFloat(
       "Vector (RGB)"=<1 1 1>
       "Float (Alpha)"=&FloatFromCurve(FloatCurve=DataInterface() {propertyValues {Curve {keys [{interpMode=RCIM_Cubic, tangentMode=RCTM_Break, time=0.5, value=1}, {interpMode=RCIM_Cubic, tangentMode=RCTM_Break, time=1, value=0}]}}})
      ))
    )) {isSetParameters=true, module=SetVariables_#3}
   &Color(Color=&MakeCustomLinearColorFromBool(
     Bool=$User.bIsPainted
     True Color=$Particles.Color
     False Color=&ColorFromCurve(ColorCurve=DataInterface() {propertyValues {
        AlphaCurve {keys [<at 0 0.6 RCIM_Cubic>,<at 1 0 RCIM_Cubic>]}
        BlueCurve {keys [<at 0 1 RCIM_Cubic>,<at 0.2373 1>,<at 0.3685 0.7379>]}
        GreenCurve {keys [<at 0 0.55 RCIM_Cubic>,<at 0.2373 0.8833>,<at 0.3685 0.75>]}
        RedCurve {keys [<at 0 0.1 RCIM_Cubic>,<at 0.2373 0.3>,<at 0.3685 0.5387>]}
       }})
    ))
   &InheritSourceMovement(
    Applied Position Scale=&Multiply_VectorByFloat(Vector=<1 1 1>, Float=1)
    ApplyForce=false
    ApplyPosition=true
    ApplyPositionFalloff=false
   )
   &SolveForcesAndVelocity(Clamp Velocity=false, Limit Acceleration=false)
   &VC_UpdateRibbon_v1(Enable Skeletal Position=true)
   &ScaleRibbonWidth(Ribbon Width Scale=&FloatFromCurve(
     FloatCurve=DataInterface() {propertyValues {Curve {keys [
         {arriveTangent=0.3269, interpMode=RCIM_Cubic, time=0, value=0.7}
         {arriveTangent=-0.6633, interpMode=RCIM_Cubic, leaveTangent=-0.6633, tangentMode=RCTM_User, time=0.25, value=0.5}
         {arriveTangentWeight=1.0346, interpMode=RCIM_Cubic, time=0.85, value=0}
        ]}}}
     Scale Curve=&SelectFloatFromArray(Direct Array Index=$Particles.SpawnGroup, Float Selection Array=DataInterface() {propertyValues {FloatData [1.25,-1.25]}})
    ))
   /SetVariables(Particles.WPO:NiagaraPosition=&AddVectorToPosition(
     Position=$Particles.Position
     Vector=&TransformVector(
      Vector=&MakeVector(
       X=0
       Y=0
       Z=&Sine(
        Normalized Angle=&Add_Float(
         A=&Multiply_Float(A=$Particles.NormalizedAge, B=1)
         B=&Multiply_Float(A=$Emitter.LoopedAge, B=-0.5)
        )
        Period=&MaskFloatBySpawnGroup(Masked Float=-0.5, Passthrough Float=0.5)
        Scale=&FloatFromCurve(
         FloatCurve=DataInterface() {propertyValues {Curve {
            defaultValue=1
            keys [{interpMode=RCIM_Cubic, tangentMode=RCTM_User, time=0, value=0}, {interpMode=RCIM_Cubic, tangentMode=RCTM_User, time=0.689, value=1}]
           }}}
         Scale Curve=3
        )
        Bias=0
       )
      )
      Destination Space=::Simulation
      Apply Scale=true
     )
    )) {isSetParameters=true}
  ]
  rendererClasses [&NiagaraRibbonRendererProperties]
  renderers [{rendererClass=&NiagaraRibbonRendererProperties}]
 }
}
summary {
 systemName=NS_Boost_Aurawater
 emitters [
  {emitterName=ribbon, rendererClasses [&NiagaraRibbonRendererProperties]}
  {emitterName=ribbon002, rendererClasses [&NiagaraRibbonRendererProperties]}
  {emitterName=StarSprites001, rendererClasses [&NiagaraSpriteRendererProperties]}
  {emitterName=ribbon003, rendererClasses [&NiagaraRibbonRendererProperties]}
  {emitterName=mesh_glow002, rendererClasses [&NiagaraMeshRendererProperties, &NiagaraSpriteRendererProperties]}
 ]
 userVariables [
  User.bIsFrontendPreview {type=NiagaraBool}
  User.bIsPainted {type=NiagaraBool}
  User.ColorVariantID {type=NiagaraInt32}
  User.DTCarOffset {type=Vector3f, description="", defaultValue {struct=Vector3f, value=<0 0 0>}}
  User.PaintedColor {type=LinearColor, description="", defaultValue {struct=LinearColor, value=<rgba 1 1 1 1>}}
  User.PaintedScalar {type=NiagaraFloat, defaultValue=1, description=""}
  User.PortCount {type=NiagaraInt32, defaultValue=2}
 ]
}
